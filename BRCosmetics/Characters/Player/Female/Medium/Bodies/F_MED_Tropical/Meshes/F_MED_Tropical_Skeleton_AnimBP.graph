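# Animation Blueprint for the F_MED_Tropical female body cosmetic that adds secondary motion to its boot strings.
# It copies the pose from the attached parent mesh, then runs two chained Anim Dynamics nodes on the bones dyn_boot_strings_01 and dyn_boot_strings_02.
# Each uses a root-relative linear spring with reduced gravity, high damping and tight linear limits, blended at partial alpha.
# The Event Graph contains no logic.
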AnimBlueprint /BRCosmetics/Characters/Player/Female/Medium/Bodies/F_MED_Tropical/Meshes/F_MED_Tropical_Skeleton_AnimBP
target_skeleton=&/F_MED_Tropical_Skeleton
graphs [
 AnimGraph {nodes [
   n1 = Anim Dynamics - Bone: dyn_boot_strings_01(ComponentPose=Local To Component(LocalPose=Copy Pose From Mesh() {bUseAttachedParent=true}), Alpha=0.55, ExternalForce="0,0,0") {
    bGravityOverrideInSimSpace=true
    bLinearSpring=true
    bOverrideLinearDamping=true
    gravityScale=0.33
    linearDampingOverride=0.995
    linearSpringConstant=150
    simulationSpace=RootRelative
    alphaBoolBlend {blendOption=Linear}
    boundBone {boneName=dyn_boot_strings_01}
    physicsBodyDefinitions [{
      sphereCollisionRadius=0
      boundBone {boneName=dyn_boot_strings_01}
      boxExtents=<10 10 10>
      constraintSetup {angularLimitsMax=<0 0 0>, angularLimitsMin=<0 0 0>, angularTarget=<0 0 0>, linearAxesMax=<4 2 0.02>, linearAxesMin=<-4 -2 -0.02>}
      localJointOffset=<0 0 -5>
     }]
   }
   n2 = Anim Dynamics - Bone: dyn_boot_strings_02(ComponentPose=@n1, Alpha=0.3, ExternalForce="0,0,0") {
    bGravityOverrideInSimSpace=true
    bLinearSpring=true
    bOverrideLinearDamping=true
    gravityScale=0.29
    linearDampingOverride=0.995
    linearSpringConstant=150
    simulationSpace=RootRelative
    alphaBoolBlend {blendOption=Linear}
    boundBone {boneName=dyn_boot_strings_02}
    physicsBodyDefinitions [{
      sphereCollisionRadius=0
      boundBone {boneName=dyn_boot_strings_02}
      boxExtents=<10 10 10>
      constraintSetup {angularLimitsMax=<0 0 0>, angularLimitsMin=<0 0 0>, angularTarget=<0 0 0>, linearAxesMax=<4 2 0.02>, linearAxesMin=<-5 -2 -0.02>}
      localJointOffset=<0 0 -6>
     }]
   }
   Output Pose(Result=Component To Local(ComponentPose=@n2)) {layerGroup=None}
  ]}
 EventGraph {nodes [Event BlueprintUpdateAnimation()]}
]
variables []
